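; Looping gameplay cue that drives a player pawn's visual effects while Nitro is active.
; Async tag watchers toggle FX visibility, set the speed FX's falling flag and manage a sprint camera lens effect, while a timer feeds remaining duration to looping particles.
; Footstep events during tactical sprint trace for surfaces to place a reused Niagara water splash and suppress dust.
; Removal unbinds footsteps, ends the watchers and releases the lens particle.
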(event EventOnApplication(EitherVFX-preferthisone)
  (params (MyTarget :type Actor%20Object%20Reference) (Parameters :type Gameplay%20Cue%20Parameters%20Structure%20%28by%20ref%29) (ParticleComponents :type Array%20of%20FXSystem%20Component%20Object%20References) (AudioComponents :type Array%20of%20Audio%20Component%20Object%20References) (BurstCameraShakeInstance :type Camera%20Shake%20Base%20Object%20Reference) (BurstDecalInstance :type Decal%20Actor%20Object%20Reference))
  (|CallParentFunction MyTarget Parameters ParticleComponents AudioComponents BurstCameraShakeInstance BurstDecalInstance))

(event EventOnLoopingStart(EitherVFX-preferthisversion)
  (params (MyTarget :type Actor%20Object%20Reference) (Parameters :type Gameplay%20Cue%20Parameters%20Structure%20%28by%20ref%29) (ParticleComponents :type Array%20of%20FXSystem%20Component%20Object%20References) (AudioComponents :type Array%20of%20Audio%20Component%20Object%20References))
  (|CallParentFunction MyTarget Parameters ParticleComponents AudioComponents)
  (Variables|Default|SetSpawnedParticleComponents ParticleComponents)
  (bind _output_get (Variables|Default|SetFortPlayerPawn (Utilities|Casting|CastToFortPlayerPawn (Variables|FortGCNL_Nitro_Active_Base|GetTargetActor))))
  (EventDispatchers|BindEventtoOnFootstepEvent _output_get (EventDispatchers|CreateEvent :!event WaterFootsteps))
  (Variables|Default|SetSpawnedLoopingParticles ParticleComponents)
  (Effects|Components|ParticleSystem|SetBoolParameter (Variables|Speed|GetSpawnedSpeedFX) "IsLocalPlayer" (Ability|GameplayCue|IsInstigatorLocallyControlled Parameters))
  (bind _asynctaskproxy (Ability|Tasks|WaitGameplayTagQueryOnActor (Variables|FortGCNL_Nitro_Active_Base|GetTargetActor) (Variables|Default|GetHideFXTagQuery))
    (:then
      (Variables|Default|SetFXNotVisibleAsyncAction _asynctaskproxy)
      (bind _asynctaskproxy_2 (Ability|Tasks|WaitGameplayTagQueryOnActor (Variables|FortGCNL_Nitro_Active_Base|GetTargetActor) (Variables|Default|GetHideFXTagQuery) "WhenFalse")
        (:then
          (Variables|Default|SetFXVisibleAsyncAction _asynctaskproxy_2)
          (bind (_asynctaskproxy_3 _tagcount) (Ability|Tasks|WaitGameplayTagCountChangedOnActor (Variables|FortGCNL_Nitro_Active_Base|GetTargetActor) (Variables|Default|GetTag%20to%20Check))
            (:then
              (Variables|Default|SetAsyncSprintTagAction _asynctaskproxy_3)
              (bind (_asynctaskproxy_4 _tagcount_2) (Ability|Tasks|WaitGameplayTagCountChangedOnActor (Variables|FortGCNL_Nitro_Active_Base|GetTargetActor) (Variables|Default|GetFallingTag))
                (:then
                  (Variables|Default|SetFallingAsyncAction _asynctaskproxy_4)
                  (bind _returnvalue (Utilities|Time|SetTimerbyEvent (delegate Custom|SetRemainingDurationParameter) 0.5 true))
                  (Variables|Default|SetTimerHandle _returnvalue))
                (:TagCountChanged
                  (CallFunction|ManageFallingTag (>= _tagcount_2 0)))))
            (:TagCountChanged
              (CallFunction|ManageSprintTag (> _tagcount 0)))))
        (:Triggered
          (CallFunction|SetFXVisible true))))
    (:Triggered
      (CallFunction|SetFXVisible false))))

(event Custom|SetRemainingDurationParameter
  (bind _result (ArrayMacros|SafeArrayGet (Variables|Default|GetSpawnedLoopingParticles))
    (:%20%20
      (Utilities|IsValid _result
        (:Is%20Valid
          (Effects|Components|ParticleSystem|SetFloatParameter _result (Variables|Default|GetDurationParamName) (Variables|FortGCNL_Nitro_Active_Base|GetRemainingDuration)))))))

(event EventOnRemoval(EitherVFX-preferthisversion)
  (params (MyTarget :type Actor%20Object%20Reference) (Parameters :type Gameplay%20Cue%20Parameters%20Structure%20%28by%20ref%29) (ParticleComponents :type Array%20of%20FXSystem%20Component%20Object%20References) (AudioComponents :type Array%20of%20Audio%20Component%20Object%20References) (BurstCameraShakeInstance :type Camera%20Shake%20Base%20Object%20Reference) (BurstDecalInstance :type Decal%20Actor%20Object%20Reference))
  (|CallParentFunction MyTarget Parameters ParticleComponents AudioComponents BurstCameraShakeInstance BurstDecalInstance)
  (EventDispatchers|UnbindEventfromOnFootstepEvent (Variables|Default|GetFortPlayerPawn) (EventDispatchers|CreateEvent :!event WaterFootsteps))
  ;[1 1 1] Making sure that everything will be cleared out.
  (bind _asyncsprinttagaction (Variables|Default|GetAsyncSprintTagAction)
    (:then
      (Ability|Async|EndAction _asyncsprinttagaction))
    (:else
      _))
  (bind _fallingasyncaction (Variables|Default|GetFallingAsyncAction)
    (:then
      (Ability|Async|EndAction _fallingasyncaction))
    (:else
      _))
  (bind _fxvisibleasyncaction (Variables|Default|GetFXVisibleAsyncAction)
    (:then
      (Ability|Async|EndAction _fxvisibleasyncaction))
    (:else
      _))
  (bind _fxnotvisibleasyncaction (Variables|Default|GetFXNotVisibleAsyncAction)
    (:then
      (Ability|Async|EndAction _fxnotvisibleasyncaction))
    (:else
      _))
  (CallFunction|DestroyCameraLensEffect))

(event Custom|DestroyCameraLensEffect
  (bind _cameralensfx (Variables|Default|GetCameraLensFX))
  (bind _returnvalue (CameraLensEffect|GetPrimaryParticleComponent _cameralensfx))
  (Utilities|IsValid (Utilities|ToObject(Interface) _cameralensfx)
    (:Is%20Valid
      (Utilities|IsValid _returnvalue
        (:Is%20Valid
          (Effects|Components|ParticleSystem|ReleasetoPool _returnvalue))))))

(fn ManageFallingTag
  (params (HasSprintTag :type Boolean))
  (bind _spawnedspeedfx (Variables|Speed|GetSpawnedSpeedFX))
  (Utilities|IsValid _spawnedspeedfx
    (:Is%20Valid
      (Effects|Components|ParticleSystem|SetBoolParameter _spawnedspeedfx "IsFalling" HasSprintTag))))

(fn ManageSprintTag
  (params (HasSprintTag :type Boolean))
  (if (not HasSprintTag)
    (CallFunction|DestroyCameraLensEffect))
  (Utilities|FlowControl|DoOnce false
    (:Completed
      (bind _returnvalue (Camera|AddGenericCameraLensEffect (Variables|/Script/Engine.PlayerController|GetPlayerCameraManager (Fortnite|GetFortPlayerControllerfromActor (Variables|Default|GetFortPlayerPawn))) "/NitroGameplay/Assets/FX/BP_CameraLens_NitroLoop.BP_CameraLens_NitroLoop_C"))
      (Variables|Default|SetCameraLensFX _returnvalue)
      (Development|PrintString 0 "I'M SPRINTING WITH NITRO"))))

(fn SetFXVisible
  (params (IsVisible :type Boolean))
  (Development|PrintString 0 (Utilities|String|Append "Nitro FX Visible: " (Utilities|String|ToString(Boolean) IsVisible)) true true (Math|Color|SelectColor "(R=0.000000,G=1.000000,B=0.528800,A=1.000000)" "(R=1.000000,G=0.250882,B=0.000000,A=1.000000)" IsVisible))
  (Utilities|FlowControl|Sequence
    (:then_0
      (bind _spawnedspeedfx (Variables|Speed|GetSpawnedSpeedFX)
        (:then
          (Rendering|SetVisibility _spawnedspeedfx (Variables|GetIsVisible) false))
        (:else)))
    (:then_1
      (for _array_element (Variables|Default|GetSpawnedLoopingParticles)
        (Rendering|SetVisibility _array_element (Variables|GetIsVisible) false)))))

(fn ConstructionScript
  (|CallParentFunction))

(fn WaterFootsteps
  (params (FootIndex :type Integer))
  (bind _fortplayerpawn (Variables|Default|GetFortPlayerPawn))
  (bind _mesh (Variables|/Script/Engine.Character|GetMesh _fortplayerpawn))
  (bind _returnvalue (Transformation|GetSocketTransform _mesh "ball_r" "RTS_World"))
  (bind _returnvalue_2 (Transformation|GetSocketTransform _mesh "ball_l" "RTS_World"))
  (bind _returnvalue_3 (Math|Transform|SelectTransform _returnvalue _returnvalue_2 (== FootIndex 1)))
  (bind _location (.location _returnvalue_3))
  (bind _returnvalue_4 (== FootIndex 1))
  (bind _fortplayerpawn_2 (Variables|Default|GetFortPlayerPawn))
  (bind _bistacticalsprinting (Variables|/Script/FortniteGame.FortPlayerPawn|GetbIsTacticalSprinting _fortplayerpawn_2))
  (bind _footstepwatersplashfx (Variables|Default|GetFootstepWaterSplashFX))
  (bind _result (ArrayMacros|SafeArrayGet (Variables|Default|GetSpawnedLoopingParticles))
    (:%20%20
      (Effects|Components|ParticleSystem|SetBoolParameter _result "IsSprinting" _bistacticalsprinting)
      (if _bistacticalsprinting
        (bind (_outhit _returnvalue_5) (Collision|LineTraceForObjects (Variables|Default|GetFortPlayerPawn) (+ _location "0.000000,0.000000,100.000000") (+ _location "0.000000,0.000000,-50.000000") (Variables|Default|GetObject%20Types)))
        (Variables|/Script/FortniteGame.FortPlayerPawn|SetbFXPlayDustOnMovement (not _returnvalue_5) (Variables|Default|GetFortPlayerPawn))
        (if _returnvalue_5
          (Variables|Default|GetFootstepWaterSplashFX
            (:then
              (Components|Activation|Activate _footstepwatersplashfx false)
              (Rendering|SetVisibility _footstepwatersplashfx true true)
              (Transformation|SetWorldLocation _footstepwatersplashfx (.Location (Collision|BreakHitResult _outhit)) false false))
            (:else
              (bind _returnvalue_6 (Niagara|SpawnSystematLocation :SystemTemplate "/NitroGameplay/Assets/FX/NS_Nitro_Footstep_WaterSplash.NS_Nitro_Footstep_WaterSplash" :Location (.Location (Collision|BreakHitResult _outhit)) :Scale "1.000000,1.000000,1.000000" :bAutoDestroy true :bAutoActivate true :PoolingMethod "AutoRelease" :bPreCullCheck true))
              (Variables|Default|SetFootstepWaterSplashFX _returnvalue_6)))
          (Niagara|NiagaraSetVector4Array (Variables|Default|GetFootstepWaterSplashFX) "SpawnList" (Utilities|Array|MakeArray (Math|Vector4|MakeVector4 (.x _location) (.y _location) (.z _location) 1.0)))
          (else
            (Components|Activation|Deactivate (Variables|Default|GetFootstepWaterSplashFX))))))))
