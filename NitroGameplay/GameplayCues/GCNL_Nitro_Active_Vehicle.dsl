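; Looping gameplay cue for active Nitro on vehicles, driving wheel and speed FX plus engine audio ducking.
; On start it spawns pooled Niagara wheel FX tuned by tire mod and vehicle-type tags, and adds a sound component that ducks engine volume via timeline.
; Delegates feed local-passenger, water, remaining-duration and tire-mod changes into FX parameters.
; Removal releases wheel FX, unbinds delegates and reverses the duck before destroying the sound component.
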(event EventOnRemoval(EitherVFX-preferthisversion)
  (params (MyTarget :type Actor%20Object%20Reference) (Parameters :type Gameplay%20Cue%20Parameters%20Structure%20%28by%20ref%29) (ParticleComponents :type Array%20of%20FXSystem%20Component%20Object%20References) (AudioComponents :type Array%20of%20Audio%20Component%20Object%20References) (BurstCameraShakeInstance :type Camera%20Shake%20Base%20Object%20Reference) (BurstDecalInstance :type Decal%20Actor%20Object%20Reference))
  ;[1 1 1] On Removal
  (|CallParentFunction MyTarget Parameters ParticleComponents AudioComponents BurstCameraShakeInstance BurstDecalInstance)
  (Utilities|FlowControl|Sequence
    (:then_0
      (Effects|Components|ParticleSystem|ReleasetoPool (Variables|Default|GetWheelFX))
      (bind _cachednitrocomponent (Variables|Default|GetCachedNitroComponent)
        (:then
          (Default|UnbindEventfromOnRemainingDurationChanged _cachednitrocomponent (EventDispatchers|CreateEvent :!event Custom|OnRemainingDurationChanged_Event)))
        (:else)))
    (:then_1
      (bind _asfort_athena_vehicle (Utilities|Casting|CastToFortAthenaVehicle MyTarget))
      (Vehicle|UnbindEventfromOnPawnExitedVehiclefromSeat _asfort_athena_vehicle (EventDispatchers|CreateEvent :!event Custom|OnPawnExitedVehicleFromSeat_Event))
      (Vehicle|UnbindEventfromOnPawnEnteredVehicleatSeat _asfort_athena_vehicle (EventDispatchers|CreateEvent :!event Custom|OnPawnEnteredVehicleAtSeat_Event)))
    (:then_2
      (bind _asfort_vehicle_interface (Utilities|Casting|CastToFortVehicleInterface MyTarget))
      (Cosmetic|UnbindEventfromOnEnteredWaterDelegate (Vehicle|Water|GetVehiclePontoonsComponent _asfort_vehicle_interface) (EventDispatchers|CreateEvent :!event Custom|OnEnteredWaterDelegate_Event))
      (Cosmetic|UnbindEventfromOnExitedWaterDelegate (Vehicle|Water|GetVehiclePontoonsComponent _asfort_vehicle_interface) (EventDispatchers|CreateEvent :!event Custom|OnExitedWaterDelegate_Event)))
    (:then_3
      (bind _nitro_vehicle_sound_component (Variables|Default|GetNitro%20Vehicle%20Sound%20Component)
        (:then
          (bind (_direction _alpha) (Timeline|EngineVolumeRemoveDuck ReverseFromEnd)
            (:Update
              (Class|BPNitroVehicleSoundComponent|DuckEngineVolume _nitro_vehicle_sound_component _alpha))
            (:Finished
              (Class|BPNitroVehicleSoundComponent|ResetEngineAudioAndDestroy _nitro_vehicle_sound_component))))
        (:else)))
    (:then_4
      (bind _targetvehicleactor (Variables|Default|GetTargetVehicleActor)
        (:then
          (Vehicle|Mods|UnbindEventfromOnModAppliedonClientDelegate _targetvehicleactor (EventDispatchers|CreateEvent :!event ApplyTireMod)))
        (:else)))))

(event EventOnLoopingStart(EitherVFX-preferthisversion)
  (params (MyTarget :type Actor%20Object%20Reference) (Parameters :type Gameplay%20Cue%20Parameters%20Structure%20%28by%20ref%29) (ParticleComponents :type Array%20of%20FXSystem%20Component%20Object%20References) (AudioComponents :type Array%20of%20Audio%20Component%20Object%20References))
  ;[1 1 1] On Application
  (|CallParentFunction MyTarget Parameters ParticleComponents AudioComponents)
  (bind _asfort_athena_vehicle (Utilities|Casting|CastToFortAthenaVehicle MyTarget))
  (Variables|Default|SetTargetVehicleActor _asfort_athena_vehicle)
  (CallFunction|SpawnWheelFX MyTarget)
  (bind _asfort_vehicle_interface (Utilities|Casting|CastToFortVehicleInterface MyTarget))
  (Variables|Default|SetTargetVehicleInterface _asfort_vehicle_interface)
  (Utilities|FlowControl|Sequence
    (:then_0
      (CallFunction|StartListeningforEnterExit)
      (CallFunction|StartListeningforDuration)
      (CallFunction|StartListeningforWater)
      (bind _nitro_vehicle_sound_component (CallFunction|AddNitroVehicleSoundComponent))
      (bind _output_get (Variables|Default|SetNitro%20Vehicle%20Sound%20Component _nitro_vehicle_sound_component))
      (bind (_direction _alpha) (Timeline|EngineVolumeApplyDuck PlayFromStart)
        (:Update
          (Class|BPNitroVehicleSoundComponent|DuckEngineVolume _output_get _alpha))))
    (:then_1
      ;;[1 1 1] Check if we already have any locally viewed passengers
      (bind _returnvalue (CallFunction|HasAnyLocalPassengers (Variables|Default|GetTargetVehicleInterface)))
      (Development|PrintString 0 (Utilities|String|Append (Utilities|String|ToString(Boolean) _returnvalue) "= Has Any Local Passengers") true true "(R=1.000000,G=0.178871,B=0.029012,A=1.000000)" 2e+01)
      (Effects|Components|ParticleSystem|SetBoolParameter (Variables|Speed|GetSpawnedSpeedFX) "HasLocalPassengers" _returnvalue))
    (:then_2
      (Vehicle|Mods|BindEventtoOnModAppliedonClientDelegate (Variables|Default|GetTargetVehicleActor) (EventDispatchers|CreateEvent :!event ApplyTireMod)))))

(event Custom|StartListeningForEnterExit
  (bind _targetvehicleactor (Variables|Default|GetTargetVehicleActor))
  ;[1 1 1] Listen for pawns entering and exiting so we can tell if we have a locally viewed client in the car
  (Vehicle|AssignOnPawnExitedVehiclefromSeat _targetvehicleactor (delegate Custom|OnPawnExitedVehicleFromSeat_Event))
  (Vehicle|AssignOnPawnEnteredVehicleatSeat _targetvehicleactor (delegate Custom|OnPawnEnteredVehicleAtSeat_Event)))

(event Custom|OnPawnExitedVehicleFromSeat_Event
  (params (Vehicle :type Fort%20Vehicle%20Interface%20Interface%20%28by%20ref%29) (PlayerPawn :type Fort%20Pawn%20Object%20Reference) (SeatIndex :type Integer))
  ;[1 1 1] Listen for pawns entering and exiting so we can tell if we have a locally viewed client in the car
  (if (Pawn|IsLocallyViewed PlayerPawn)
    (CallFunction|OnLocalPawnPassengerChanged false)
    (Effects|Components|ParticleSystem|SetBoolParameter (Variables|Speed|GetSpawnedSpeedFX) "HasLocalPassengers" false)))

(event Custom|OnPawnEnteredVehicleAtSeat_Event
  (params (Vehicle :type Fort%20Vehicle%20Interface%20Interface%20%28by%20ref%29) (PlayerPawn :type Fort%20Pawn%20Object%20Reference) (SeatIndex :type Integer))
  ;[1 1 1] Listen for pawns entering and exiting so we can tell if we have a locally viewed client in the car
  (if (Pawn|IsLocallyViewed PlayerPawn)
    (CallFunction|OnLocalPawnPassengerChanged true)
    (Effects|Components|ParticleSystem|SetBoolParameter (Variables|Speed|GetSpawnedSpeedFX) "HasLocalPassengers" true)))

(event Custom|OnLocalPawnPassengerChanged
  (params (HasLocalPassenger :type Boolean))
  (bind _returnvalue (Utilities|String|ToString(Boolean) HasLocalPassenger))
  ;[1 1 1] Listen for pawns entering and exiting so we can tell if we have a locally viewed client in the car
  (Development|PrintString 0 _returnvalue)
  (Development|PrintString 0 (Utilities|String|Append _returnvalue "= Has Local Passengers") true true "(R=1.000000,G=0.416440,B=0.000000,A=1.000000)" 2e+01))

(event Custom|StartListeningForDuration
  (bind _nitrocomponentclass (Variables|Default|GetNitroComponentClass))
  (bind _targetvehicleactor (Variables|Default|GetTargetVehicleActor))
  (bind _returnvalue (Actor|GetComponentbyClass _targetvehicleactor _nitrocomponentclass))
  ;[1 1 1] Listen for remaining effect duration
  (Utilities|IsValid _returnvalue
    (:Is%20Valid
      (bind _asnitro_component_vehicle (Utilities|Casting|CastToNitroComponent_Vehicle _returnvalue))
      (bind _output_get (Variables|Default|SetCachedNitroComponent _asnitro_component_vehicle))
      (Default|AssignOnRemainingDurationChanged _output_get (EventDispatchers|CreateEvent :!event Custom|OnRemainingDurationChanged_Event)))))

(event Custom|OnRemainingDurationChanged_Event
  (params (NewDuration :type Float%20%28double-precision%29))
  (bind _wheelfx (Variables|Default|GetWheelFX))
  ;[1 1 1] Listen for remaining effect duration
  (CallFunction|UpdateRemainingDuration NewDuration)
  (Effects|Components|ParticleSystem|SetFloatParameter _wheelfx "RemainingDuration" NewDuration)
  (Effects|Components|ParticleSystem|SetBoolParameter _wheelfx "IsInAir" (Vehicle|VehicleIsinAir (Variables|Default|GetTargetVehicleActor)))
  (Development|PrintString 0 (Utilities|String|Append "Remaining duration is: " (Utilities|String|ToString(Float) NewDuration))))

(event Custom|StartListeningForWater
  (bind _targetvehicleinterface (Variables|Default|GetTargetVehicleInterface))
  (bind _returnvalue (Vehicle|Water|GetVehiclePontoonsComponent _targetvehicleinterface))
  ;[1 1 1] Listen for enter/exit water
  (Cosmetic|AssignOnEnteredWaterDelegate _returnvalue (delegate Custom|OnEnteredWaterDelegate_Event))
  (Cosmetic|AssignOnExitedWaterDelegate _returnvalue (delegate Custom|OnExitedWaterDelegate_Event)))

(event Custom|OnEnteredWaterDelegate_Event
  (params (Pontoon :type Spherical%20Pontoon%20Structure%20%28by%20ref%29))
  ;[1 1 1] Listen for enter/exit water
  (CallFunction|OninWaterChanged true))

(event Custom|OnExitedWaterDelegate_Event
  (params (Pontoon :type Spherical%20Pontoon%20Structure%20%28by%20ref%29))
  ;[1 1 1] Listen for enter/exit water
  (CallFunction|OninWaterChanged false))

(event Custom|OnInWaterChanged
  (params (IsInWater :type Boolean))
  (bind _wheelfx (Variables|Default|GetWheelFX))
  ;[1 1 1] Listen for enter/exit water
  (if IsInWater
    (Development|PrintString 0 "Nitro vehicle: Entered water")
    (Effects|Components|ParticleSystem|SetBoolParameter _wheelfx "IsInWater" true)
    (else
      (Development|PrintString 0 "Nitro vehicle: exited water")
      (Effects|Components|ParticleSystem|SetBoolParameter _wheelfx "IsInWater" false))))

(fn AddNitroVehicleSoundComponent
  (Utilities|IsValid (Actor|GetComponentbyClass (Variables|Default|GetTargetVehicleActor) "/NitroGameplay/Assets/Audio/BP_NitroVehicleSoundComponent.BP_NitroVehicleSoundComponent_C")
    (:Is%20Not%20Valid
      (bind _returnvalue (Game|AddComponentbyClass (Variables|Default|GetTargetVehicleActor) 0 "/NitroGameplay/Assets/Audio/BP_NitroVehicleSoundComponent.BP_NitroVehicleSoundComponent_C"))
      (return _returnvalue))))

(fn ApplyTireMod
  (params (Vehicle :type Actor%20Object%20Reference) (ModTag :type Gameplay%20Tag%20Structure%20%28by%20ref%29))
  (bind _wheelfx (Variables|Default|GetWheelFX))
  (switch Utilities|FlowControl|Switch|SwitchonGameplayTag ModTag
    (:Vehicle.Mod.Tire.Bulletproof
      (Variables|Default|SetHasTireModBulletproof true)
      (Variables|Default|SetHasTireModOffroad false))
    (:Vehicle.Mod.Tire.Offroad
      (Variables|Default|SetHasTireModOffroad true)
      (Variables|Default|SetHasTireModBulletproof false)))
  (Effects|Components|ParticleSystem|SetBoolParameter _wheelfx "HasTireModBulletproof" (Variables|Default|GetHasTireModBulletproof))
  (Effects|Components|ParticleSystem|SetBoolParameter _wheelfx "HasTireModOffroad" (Variables|Default|GetHasTireModOffroad))
  (return))

(fn HasAnyLocalPassengers
  (params (Target :type Fort%20Vehicle%20Interface%20Interface))
  (for _array_element (Vehicle|GetAllPassengers Target)
    (if (Pawn|IsLocallyViewed _array_element)
      (return true)))
  (return false))

(fn SpawnWheelFX
  (params (TargetActor :type Actor%20Object%20Reference))
  (bind _spawnedspeedfx (Variables|Speed|GetSpawnedSpeedFX))
  (bind _wheelfx (Variables|Default|GetWheelFX))
  (bind _wheelfx_2 (Variables|Default|GetWheelFX))
  (bind _wheelfx_3 (Variables|Default|GetWheelFX))
  (bind _wheelfx_4 (Variables|Default|GetWheelFX))
  (bind _returnvalue (Effects|SpawnFXSystemAttached :FXSystemAsset "/NitroGameplay/Assets/FX/NS_NitroState_VehicleV2.NS_NitroState_VehicleV2" :AttachToComponent (Variables|/Script/Engine.Actor|GetRootComponent (Variables|FortGCNL_Nitro_Active_Base|GetTargetActor)) :AttachPointName "None" :Scale "1.000000,1.000000,1.000000" :LocationType "KeepRelativeOffset" :bAutoDestroy true :PoolingMethod "ManualRelease" :bAutoActivate true))
  (Variables|Default|SetWheelFX _returnvalue)
  (Utilities|FlowControl|Sequence
    (:then_0
      ;[1 1 1] TireMod Check
      (if (Vehicle|Mod|HasVehicleModApplied (Variables|Default|GetTargetVehicleActor) (GameplayTags|MakeLiteralGameplayTag "(TagName=\"Vehicle.Mod.Tire.Bulletproof\")"))
        (Variables|Default|SetHasTireModBulletproof true)
        (Variables|Default|SetHasTireModOffroad false)
        (elif (Vehicle|Mod|HasVehicleModApplied (Variables|Default|GetTargetVehicleActor) (GameplayTags|MakeLiteralGameplayTag "(TagName=\"Vehicle.Mod.Tire.Offroad\")"))
          (Variables|Default|SetHasTireModBulletproof false)
          (Variables|Default|SetHasTireModOffroad true)
          (else
            (Variables|Default|SetHasTireModBulletproof false)
            (Variables|Default|SetHasTireModOffroad false))))
      (Effects|Components|ParticleSystem|SetBoolParameter _wheelfx "HasTireModBulletproof" (Variables|Default|GetHasTireModBulletproof))
      (Effects|Components|ParticleSystem|SetBoolParameter _wheelfx "HasTireModOffroad" (Variables|Default|GetHasTireModOffroad)))
    (:then_1
      (bind _returnvalue_2 (GameplayTags|HasMatchingGameplayTag(Message) (Variables|FortGCNL_Nitro_Active_Base|GetTargetActor) (Variables|Default|GetBattleBusTag)))
      (if _returnvalue_2
        ;[1 1 1] Handle battle bus
        (Effects|Components|ParticleSystem|SetFloatParameter _wheelfx_2 "WheelRadiusMultiplier" (select (Variables|Default|GetHasTireModOffroad) 1.3 1.9))
        (Effects|Components|ParticleSystem|SetFloatParameter _wheelfx_2 "WheelOffsetMultiplier" 1.2)
        (Effects|Components|ParticleSystem|SetFloatParameter _wheelfx_2 "WheelWidthMultiplier" 1.5)
        (Effects|Components|ParticleSystem|SetFloatParameter _wheelfx_2 "WheelOffset" -5.0)
        (Effects|Components|ParticleSystem|SetFloatParameter _spawnedspeedfx "EffectForwardOffset" 5e+02)
        (Effects|Components|ParticleSystem|SetFloatParameter _spawnedspeedfx "EffectUpwardOffset" 1e+02)
        (Effects|Components|ParticleSystem|SetFloatParameter _spawnedspeedfx "BattleBusFXOpacity" 0.1)
        (else
          ;[1 1 1] Handle motorcycles
          (bind _returnvalue_3 (GameplayTags|HasMatchingGameplayTag(Message) (Variables|FortGCNL_Nitro_Active_Base|GetTargetActor) (Variables|Default|GetMotorcycleTag)))
          (if _returnvalue_3
            (Effects|Components|ParticleSystem|SetBoolParameter (Variables|Default|GetWheelFX) "IsMotorcycle" true)
            (else
              ;[1 1 1] Is SUV
              (bind _returnvalue_4 (GameplayTags|HasMatchingGameplayTag(Message) (Variables|FortGCNL_Nitro_Active_Base|GetTargetActor) (Variables|Default|GetBasicSUVTag)))
              (if _returnvalue_4
                (Effects|Components|ParticleSystem|SetFloatParameter _wheelfx_3 "WheelRadiusMultiplier" 1.1) ; Basic SUV
                (Effects|Components|ParticleSystem|SetFloatParameter _wheelfx_3 "WheelWidthMultiplier" 0.7)
                (Effects|Components|ParticleSystem|SetBoolParameter _wheelfx_3 "IsSuv" true)
                (bind _returnvalue_5 (GameplayTags|HasMatchingGameplayTag(Message) (Variables|FortGCNL_Nitro_Active_Base|GetTargetActor) (Variables|Default|GetBossSUVTag)))
                (if _returnvalue_5
                  (Effects|Components|ParticleSystem|SetFloatParameter _wheelfx_4 "WheelRadiusMultiplier" 1.0)
                  (Effects|Components|ParticleSystem|SetFloatParameter _wheelfx_4 "WheelWidthMultiplier" 1.45))))))))))

(fn ConstructionScript
  (|CallParentFunction))
